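# Niagara effect for the QuirkLuster pickaxe jab: a warm gold burst of airflow, swirling meshes and feathers.
# Airflow and Soulercoasters each burst two spawn groups that switch per-group settings and set VisibilityTag across two renderers, animated via dynamic material parameters.
# Feathers emit sub-UV sprites after a short delay, driven by drag, curl noise and gravity.
# A time-of-day scalar modulates brightness or alpha, and Soulercoasters' burst count varies by platform quality.
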NiagaraSystem /BRCosmetics/Weapons/FORT_Melee/Pickaxe_QuirkLuster/FX/NS_Pickaxe_QuirkLuster_Jab
dependencies {
 usedDataInterfaces [
  &NiagaraDataInterfaceCurve
  &NiagaraDataInterfaceColorCurve
  &NiagaraDataInterfacePlatformSet
  &NiagaraDataInterfaceSpriteRendererInfo
 ]
 usedDynamicInputs [&FNTimeOfDayScalar]
 usedModules [
  &self:SystemSpawnScript.SystemScriptSource.SystemScriptGraph.NiagaraNodeAssignment_0.SetVariables_#1
  &SystemState
 ]
}
emitters {
 Airflow {
  emitterUpdateScript [
   &EmitterState()
   &SpawnBurst_Instantaneous(Use Spawn Probability=false, UseLoopCountLimit=false)
   &SpawnBurst_Instantaneous(Spawn Group=1, Use Spawn Probability=false, UseLoopCountLimit=false) {module=SpawnBurst_Instantaneous001}
  ]
  particleSpawnScript [
   &InitializeParticle(
    Color=<rgba 15 15 15 1>
    Lifetime=&MergeHelper_FloatSwitch(0=0.35, 1=0.2, 2=?, 3=?, Index=$Emitter.SpawnGroup, Num=::NewEnumerator0)
    Mesh Scale=<0.585 1 1>
    Mesh Scale Mode=::NewEnumerator4
    Position Offset=&MergeHelper_VectorSwitch(0=<0 0 100>, 1=<0 0 -15>, 2=?, 3=?, Index=$Emitter.SpawnGroup, Num=::NewEnumerator0)
    Sprite Size=<35 100>
    Sprite Size Mode=::NewEnumerator4
    UsePositionOffset=true
    Write Material Random=true
   )
   &AddVelocity(Velocity=&MakeVector(
     X=0
     Y=0
     Z=&MergeHelper_FloatSwitch(0=50, 1=465, 2=?, 3=?, Index=$Emitter.SpawnGroup, Num=::NewEnumerator0)
    ))
   &InitialMeshOrientation(Mesh Orientation Mode=::NewEnumerator2, Rotation=<0 -0.25 0>, Use Rotation Vector=true)
   &CameraOffset(Camera Offset Amount=&MergeHelper_FloatSwitch(0=20, 1=5, 2=?, 3=?, Index=$Emitter.SpawnGroup, Num=::NewEnumerator0))
   /SetVariables(Particles.VisibilityTag=$Emitter.SpawnGroup) {isSetParameters=true}
  ]
  particleUpdateScript [
   &ParticleState()
   &ScaleMeshSize(Scale Factor=&MakeVector(
     X=&FloatFromCurve(FloatCurve=DataInterface() {propertyValues {Curve {keys [<at 0 1.5>,<at 1 1>]}}})
     Y=1
     Z=1
    ))
   &SolveForcesAndVelocity(Clamp Velocity=false, Limit Acceleration=false)
   &DynamicMaterialParameters(
    Index 0 Param 1=&FloatFromCurve(FloatCurve=DataInterface() {propertyValues {Curve {keys [<at 0 0>,<at 0.5 0>,<at 1 0.08>]}}})
    Index 0 Param 2=&FloatFromCurve(FloatCurve=DataInterface() {propertyValues {Curve {keys [{arriveTangent=-2.2616, interpMode=RCIM_Cubic, time=0, value=-0.5}, {interpMode=RCIM_Cubic, leaveTangent=-0.7156, time=1, value=1}]}}})
    Index 0 Param 3=1.2
    Index 0 Param 4=&FloatFromCurve(FloatCurve=DataInterface() {propertyValues {Curve {keys [<at 0 1>,<at 1 2>]}}}, Scale Curve=0.2)
    Module.Index0Float0Write=true
    Module.Index0Float1Write=true
    Module.Index0Float2Write=true
    Module.Index0Float3Write=true
   )
   &ScaleColor(
    Scale Alpha=&Multiply_Float(
     A=&FloatFromCurve(
      FloatCurve=DataInterface() {propertyValues {Curve {keys [<at 0 1 RCIM_Cubic>,<at 1 0 RCIM_Cubic>]}}}
      Scale Curve=&Multiply_Float(A=0.5, B=$System.ToDScalar)
     )
     B=&FloatFromCurve(FloatCurve=DataInterface() {propertyValues {Curve {keys [<at 0 0 RCIM_Cubic>,<at 0.175 1 RCIM_Cubic>]}}}, CurveIndex=$Engine.System.Age)
    )
    Scale RGB=&MakeVectorFromLinearColorRGB(Color=&ColorFromCurve(ColorCurve=DataInterface() {propertyValues {
        AlphaCurve {keys [<at 0 1 RCIM_Cubic>]}
        BlueCurve {keys [<at 0 0.26 RCIM_Cubic>,<at 1 0 RCIM_Cubic>]}
        GreenCurve {keys [<at 0 2.4417 RCIM_Cubic>,<at 1 0.4167 RCIM_Cubic>]}
        RedCurve {keys [<at 0 4 RCIM_Cubic>,<at 1 1 RCIM_Cubic>]}
       }}))
    ScaleA=true
    ScaleRGB=true
   )
   &FN_Sprite_Attributes(
    Color=&MakeLinearColorFromVectorAndFloat(
     "Vector (RGB)"=&Multiply_VectorByFloat(
      Vector=&MakeVectorFromLinearColorRGB(Color=&ColorFromCurve(ColorCurve=DataInterface() {propertyValues {
          AlphaCurve {keys [<at 0 1 RCIM_Cubic>]}
          BlueCurve {keys [<at 0 0 RCIM_Cubic>,<at 1 0.195 RCIM_Cubic>]}
          GreenCurve {keys [<at 0 0.5 RCIM_Cubic>,<at 1 1.8312 RCIM_Cubic>]}
          RedCurve {keys [<at 0 1 RCIM_Cubic>,<at 1 3 RCIM_Cubic>]}
         }}))
      Float=$System.ToDScalar
     )
     "Float (Alpha)"=&FloatFromCurve(FloatCurve=DataInterface() {propertyValues {Curve {keys [<at 0 0>,<at 0.1 1>,<at 0.7 1>,<at 1 0>]}}})
    )
    Create Color Attribute=true
    Write to Module Namespace=true
   )
  ]
  rendererClasses [&NiagaraMeshRendererProperties, &NiagaraSpriteRendererProperties]
  renderers [
   {rendererClass=&NiagaraMeshRendererProperties, rendererIndex=0}
   {rendererClass=&NiagaraSpriteRendererProperties, rendererIndex=1}
  ]
 }
 Feathers {
  emitterUpdateScript [
   &EmitterState(Life Cycle Mode=::NewEnumerator1, Loop Behavior=::NewEnumerator1, Loop Delay=0.08, Loop Duration=0.225, UseLoopDelay=true)
   &SpawnRate(SpawnRate=12, Use Spawn Probability=false)
   &SpawnBurst_Instantaneous(
    Spawn Count=&RandomRangeInt(Minimum=1, Maximum=2)
    Use Spawn Probability=false
    UseLoopCountLimit=false
   )
  ]
  particleSpawnScript [
   &InitializeParticle(
    AdjustAlpha=false
    AdjustHue=true
    AdjustSaturation=false
    AdjustValue=false
    Color=&RandomRangeLinearColor(Minimum=<rgba 0.6 0.3675 0.135 1>, Maximum=<rgba 0.6 0.4863 0.1452 1>)
    Color Mode=::NewEnumerator3
    Hue Shift Range:Vector2f=<-0.01 0.01>
    Lifetime Max=0.7
    Lifetime Min=0.45
    Lifetime Mode=::NewEnumerator1
    Saturation Range:Vector2f=<0.4 1>
    Sprite Rotation Mode=::NewEnumerator2
    Sprite Size Mode=::NewEnumerator1
    Sprite UV Mode=::NewEnumerator3
    Uniform Sprite Size Max=16
    Uniform Sprite Size Min=8.5
    UsePositionOffset=false
    Write Material Random=true
   )
   &ShapeLocation(
    MaskSpawnGroup=false
    Offset=<0 0 10>
    Offset Coordinate Space=::Local
    Offset Mode=::NewEnumerator0
    Override Seed=false
    Sphere Radius=10
    UseOwnerScale=true
   )
   &CameraOffset(Camera Offset Amount=20)
   &AddVelocity(
    Velocity=&MakeVector(X=0, Y=0, Z=&RandomRangeFloat(Minimum=250, Maximum=325))
    Velocity Speed Scale=2
   )
   &SubUVAnimation(
    End Frame Range Override=2
    Evaluation Type=::NewEnumerator0
    Output Alternate Renderer Binding=true
    "Sprite Renderer "=DataInterface() {propertyValues {SpriteRenderer=&self:Feathers_0.NiagaraSpriteRendererProperties_1}}
    Start Frame Range Override=1
    SubUV Animation Mode:ENiagara_SubUVLookupModeV2=::NewEnumerator2
    UseEndFrame=true
    UseStartFrame=true
   )
  ]
  particleUpdateScript [
   &ParticleState()
   &Drag(Drag=&RandomRangeFloat(Minimum=5, Maximum=8), Use Linear Drag=true, Use Rotational Drag=true)
   &ScaleColor(
    Scale RGB=&VectorFromFloat(Value=&FloatFromCurve(
      FloatCurve=DataInterface() {propertyValues {Curve {keys [{arriveTangent=-2.7905, interpMode=RCIM_Cubic, leaveTangent=-2.7905, tangentMode=RCTM_User, time=0, value=2}, <at 1 1 RCIM_Cubic>]}}}
      Scale Curve=$System.ToDScalar
     ))
    ScaleA=true
    ScaleRGB=true
   )
   &ScaleSpriteSize(Scale Factor=&Vector2DFromFloat(Value=&FloatFromCurve(FloatCurve=DataInterface() {propertyValues {Curve {keys [
          {arriveTangent=0.3685, interpMode=RCIM_Cubic, time=0, value=0.4}
          {arriveTangentWeight=0.1202, interpMode=RCIM_Cubic, leaveTangentWeight=0.1202, tangentMode=RCTM_User, time=0.2, value=1}
          {arriveTangentWeight=0.1067, interpMode=RCIM_Cubic, leaveTangentWeight=0.3202, tangentMode=RCTM_User, time=0.4, value=1}
          {arriveTangent=-1.9694, arriveTangentWeight=0.2333, interpMode=RCIM_Cubic, leaveTangent=-1.9694, tangentMode=RCTM_User, time=1, value=0}
         ]}}})))
   &SpriteRotationRate(Rotation Rate=&Multiply_Float(
     A=&UniformAOrBFloat(A=-1, Evaluation Type=::NewEnumerator0, Recalculate Random Each Loop=true, Override Seed=false)
     B=&RandomRangeFloat(Minimum=110, Maximum=165)
    ))
   &FN_Sprite_Attributes(
    Color=&MakeLinearColorFromVectorAndFloat(
     "Vector (RGB)"=&MakeVectorFromLinearColorRGB(Color=&ColorFromCurve(ColorCurve=DataInterface() {propertyValues {
         AlphaCurve {keys []}
         BlueCurve {keys [<at 0 0.1302 RCIM_Cubic>,<at 1 0 RCIM_Cubic>]}
         GreenCurve {keys [<at 0 0.8122 RCIM_Cubic>,<at 1 0.4095 RCIM_Cubic>]}
         RedCurve {keys [<at 0 1 RCIM_Cubic>,<at 1 1 RCIM_Cubic>]}
        }}))
     "Float (Alpha)"=0.05
    )
    Create Color Attribute=true
    Create SpriteSize Attribute=true
    SpriteSize=&Multiply_Vector2D(A=$Particles.SpriteSize, B=<2.5 2.5>)
    Write to Module Namespace=true
   )
   &CurlNoiseForce(
    Bias Noise=false
    Noise Frequency=20
    Noise Strength=&RandomRangeFloat(Minimum=700, Maximum=1000)
    Pan Noise=false
   )
   &GravityForce(Gravity=&MakeVector(X=0, Y=0, Z=&RandomRangeFloat(Minimum=-350, Maximum=-500)))
   &SolveForcesAndVelocity(Clamp Velocity=false, Limit Acceleration=false)
  ]
  rendererClasses [&NiagaraSpriteRendererProperties]
  renderers [
   {rendererClass=&NiagaraSpriteRendererProperties}
   {rendererClass=&NiagaraSpriteRendererProperties, rendererIndex=1}
  ]
 }
 Soulercoasters {
  emitterUpdateScript [
   &EmitterState()
   &SpawnBurst_Instantaneous(
    Spawn Count=&MakeIntFromFloat(Float=&FN_PickFloatDependOnPlatform(TargetPlatform=DataInterface() {propertyValues {Platforms {qualityLevelMask=-29}}}, TargetPlatformActive=3, TargetPlatformNoActive=5))
    Use Spawn Probability=false
    UseLoopCountLimit=false
   )
   &SpawnBurst_Instantaneous(Spawn Group=1, Spawn Time=0.085, Use Spawn Probability=false, UseLoopCountLimit=false) {module=SpawnBurst_Instantaneous001}
  ]
  particleSpawnScript [
   &InitializeParticle(
    Color=<rgba 5 5 5 1>
    Lifetime=&MergeHelper_FloatSwitch(0=0.25, 1=0.2, 2=?, 3=?, Index=$Emitter.SpawnGroup, Num=::NewEnumerator0)
    Mesh Scale=&MergeHelper_VectorSwitch(
     0=&MakeVector(X=1, Y=1, Z=&RandomRangeFloat(Minimum=1.1, Maximum=1.5))
     1=<0.1 0.075 0.075>
     2=?
     3=?
     Index=$Emitter.SpawnGroup
     Num=::NewEnumerator0
    )
    Mesh Scale Mode=::NewEnumerator4
    Position Offset=&MergeHelper_VectorSwitch(0=<0 0 110>, 1=<0 0 100>, 2=?, 3=?, Index=$Emitter.SpawnGroup, Num=::NewEnumerator0)
    UsePositionOffset=true
    Write Material Random=true
   )
   &AddRotationalVelocity(Rotation Rate=<0 0 -0.125>)
   &InitialMeshOrientation(
    Mesh Orientation Mode=::NewEnumerator7
    Rotation=&MergeHelper_VectorSwitch(
     0=&MakeVector(X=0, Y=0, Z=&ReturnNormalizedExecIndex(Normalized Index Includes One=false))
     1=<0 -0.25 0>
     2=?
     3=?
     Index=$Emitter.SpawnGroup
     Num=::NewEnumerator0
    )
    Use Rotation Vector=true
   )
   /SetVariables(Particles.VisibilityTag=$Emitter.SpawnGroup) {isSetParameters=true}
   &CameraOffset(Camera Offset Amount=20)
  ]
  particleUpdateScript [
   &ParticleState()
   &ScaleMeshSize(Scale Factor=&MergeHelper_VectorSwitch(0=<1 1 1>, 1=<1 1 1>, 2=?, 3=?, Index=$Particles.VisibilityTag, Num=::NewEnumerator0))
   &ScaleColor(
    Scale Alpha=&FloatFromCurve(
     FloatCurve=DataInterface() {propertyValues {Curve {keys [{arriveTangent=-2.9154, interpMode=RCIM_Cubic, leaveTangent=-2.9154, tangentMode=RCTM_User, time=0, value=1}, <at 1 0>]}}}
     Scale Curve=$System.ToDScalar
    )
    Scale RGB=&MergeHelper_VectorSwitch(
     0=&MakeVectorFromLinearColorRGB(Color=&ColorFromCurve(ColorCurve=DataInterface() {propertyValues {
         AlphaCurve {keys [{interpMode=RCIM_Cubic, tangentMode=RCTM_User, time=0, value=1}]}
         BlueCurve {keys [
           {arriveTangentWeight=0.0472, interpMode=RCIM_Cubic, leaveTangentWeight=0.0472, tangentMode=RCTM_User, time=0.0259, value=0.069}
           {tangentMode=RCTM_User, time=0.1676, value=0.0347}
           {arriveTangentWeight=0.2775, interpMode=RCIM_Cubic, leaveTangentWeight=0.2775, tangentMode=RCTM_User, time=0.7427, value=0}
          ]}
         GreenCurve {keys [
           {arriveTangentWeight=0.0472, interpMode=RCIM_Cubic, leaveTangentWeight=0.0472, tangentMode=RCTM_User, time=0.0259, value=0.3787}
           {tangentMode=RCTM_User, time=0.1676, value=0.3353}
           {arriveTangentWeight=0.2775, interpMode=RCIM_Cubic, leaveTangentWeight=0.2775, tangentMode=RCTM_User, time=0.7427, value=0.1875}
          ]}
         RedCurve {keys [
           {arriveTangentWeight=0.0472, interpMode=RCIM_Cubic, leaveTangentWeight=0.0472, tangentMode=RCTM_User, time=0.0259, value=0.6}
           {tangentMode=RCTM_User, time=0.1676, value=0.55}
           {arriveTangentWeight=0.2775, interpMode=RCIM_Cubic, leaveTangentWeight=0.2775, tangentMode=RCTM_User, time=0.7427, value=0.45}
          ]}
        }}))
     1=&MakeVectorFromLinearColorRGB(Color=&ColorFromCurve(ColorCurve=DataInterface() {propertyValues {
         AlphaCurve {keys [<at 0 1 RCIM_Cubic>]}
         BlueCurve {keys [<at 0.1878 0.065>, {arriveTangent=-0.08, interpMode=RCIM_Cubic, leaveTangent=-0.08, time=0.6331, value=0.065}, <at 1 0 RCIM_Cubic>]}
         GreenCurve {keys [<at 0.1878 0.6104>, {arriveTangent=-0.4438, interpMode=RCIM_Cubic, leaveTangent=-0.4438, time=0.6331, value=0.6104}, <at 1 0.25 RCIM_Cubic>]}
         RedCurve {keys [<at 0.1878 1>, {arriveTangent=-0.4925, interpMode=RCIM_Cubic, leaveTangent=-0.4925, time=0.6331, value=1}, <at 1 0.6 RCIM_Cubic>]}
        }}))
     2=?
     3=?
     Index=$Particles.VisibilityTag
     Num=::NewEnumerator0
    )
    ScaleA=true
    ScaleRGB=true
   )
   &DynamicMaterialParameters(
    Index 0 Param 1=&FloatFromCurve(FloatCurve=DataInterface() {propertyValues {Curve {keys [{arriveTangent=2.8243, interpMode=RCIM_Cubic, leaveTangent=2.8243, tangentMode=RCTM_User, time=0, value=-0.872}, <at 1 0.7557>]}}})
    Index 0 Param 2=1
    Index 0 Param 4=1
    Module.Index0Float0Write=true
    Module.Index0Float1Write=true
    Module.Index0Float2Write=true
    Module.Index0Float3Write=true
   )
   /SetVariables(Particles.DynamicParametersAlt:Vector4f=&MakeVector4(
     X=&FloatFromCurve(FloatCurve=DataInterface() {propertyValues {Curve {keys [<at 0 0>,<at 1 0.3046>]}}})
     Y=1
     Z=0
     W=1
    )) {isSetParameters=true}
   &SolveForcesAndVelocity(Clamp Velocity=false, Limit Acceleration=false)
  ]
  rendererClasses [&NiagaraMeshRendererProperties]
  renderers [
   {rendererClass=&NiagaraMeshRendererProperties, rendererIndex=0}
   {rendererClass=&NiagaraMeshRendererProperties, rendererIndex=1}
  ]
 }
}
summary {
 systemName=NS_Pickaxe_QuirkLuster_Jab
 emitters [
  {emitterName=Airflow, rendererClasses [&NiagaraMeshRendererProperties, &NiagaraSpriteRendererProperties]}
  {emitterName=Soulercoasters, rendererClasses [&NiagaraMeshRendererProperties]}
  {emitterName=Feathers, rendererClasses [&NiagaraSpriteRendererProperties]}
 ]
 userVariables [
  User.bIsFrontEnd {type=NiagaraBool}
  User.bIsFrontEndPreview {type=NiagaraBool}
  User.SwingSectionIndex {type=NiagaraInt32}
 ]
}
